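; Burst gameplay cue for firing the Nitro Queen weapon in MoonflaxMedallions, which places its effect at the weapon's muzzle.
; On burst, it casts the target to a Fort player pawn and gets the current weapon.
; If that weapon's skeletal mesh asset is valid, it moves the first spawned particle component to the muzzle's world location, setting position only, without attaching it.
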(fn OnBurst(EitherVFX-preferthisversion)
  (params (MyTarget :type Actor%20Object%20Reference) (Parameters :type Gameplay%20Cue%20Parameters%20Structure%20%28by%20ref%29) (ParticleComponents :type Array%20of%20FXSystem%20Component%20Object%20References) (AudioComponents :type Array%20of%20Audio%20Component%20Object%20References) (BurstCameraShakeInstance :type Camera%20Shake%20Base%20Object%20Reference) (BurstDecalInstance :type Decal%20Actor%20Object%20Reference))
  (Utilities|FlowControl|Sequence
    (:then_0
      (bind _asfort_player_pawn (Utilities|Casting|CastToFortPlayerPawn MyTarget))
      (bind _output_get (LocalVariables|SetCurrentWeapon (Weapon|GetCurrentWeapon _asfort_player_pawn)))
      (Utilities|IsValid (Components|SkeletalMesh|GetSkeletalMeshAsset (Variables|/Script/FortniteGame.FortWeapon|GetWeaponMesh _output_get))
        (:Is%20Valid
          (Transformation|SetWorldLocation (Utilities|Array|Get(acopy) ParticleComponents) (Effects|GetMuzzleLocation _output_get -1) false false))))
    (:then_1
      (return))))
